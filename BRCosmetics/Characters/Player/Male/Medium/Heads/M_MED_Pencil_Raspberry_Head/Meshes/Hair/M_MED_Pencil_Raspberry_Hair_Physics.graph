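# Physics asset for the M_MED_Pencil_Raspberry male player head's hair, setting up secondary motion on front, side, back, top and cowlick locks.
# Each lock has a capsule on its dyn_hair bone, constrained by narrow swing and twist limits to a small MSTR anchor capsule.
# MSTR anchors attach to the two-capsule head body with free swing and twist.
# Only front_01 chains further, hanging off front_00 instead of its own anchor.
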PhysicsAsset /BRCosmetics/Characters/Player/Male/Medium/Heads/M_MED_Pencil_Raspberry_Head/Meshes/Hair/M_MED_Pencil_Raspberry_Hair_Physics
bodies {
 dyn_hair_backTop_00 [{
   shapeType=Capsule
   length=1
   radius=1
   shapeName=dyn_hair_cowlick_r_03_capsule
   center=<1.349 0.2006 0.3102>
   rotation=<rot 11.6606 89.9662 104.4079>
  }]
 dyn_hair_backTop_00_MSTR [{shapeType=Capsule, length=0.505, radius=0.505, shapeName=dyn_hair_backTop_00_MSTR_capsule, rotation=<rot 0 -90 0>}]
 dyn_hair_backTop_01 [{
   shapeType=Capsule
   length=1
   radius=1
   shapeName=dyn_hair_cowlick_r_03_capsule
   center=<1.1336 0.5522 0.3262>
   rotation=<rot 15.4045 72.5188 100.2931>
  }]
 dyn_hair_backTop_01_MSTR [{shapeType=Capsule, length=0.505, radius=0.505, shapeName=dyn_hair_backTop_01_MSTR_capsule, rotation=<rot 0 -90 0>}]
 dyn_hair_back_l_00_MSTR_Rasp [{shapeType=Capsule, length=0.505, radius=0.505, shapeName=dyn_hair_back_l_00_MSTR_Rasp_capsule, rotation=<rot 0 -90 0>}]
 dyn_hair_back_l_00_Rasp [{
   shapeType=Capsule
   length=1
   radius=1
   shapeName=dyn_hair_cowlick_r_03_capsule
   center=<2.4351 -0.3702 0.7203>
   rotation=<rot 16.3016 -8.6454 -8.8202>
  }]
 dyn_hair_back_l_01 [{shapeType=Capsule, length=1, radius=1, shapeName=dyn_hair_cowlick_r_03_capsule, center=<2.5992 0 0.2507>, rotation=<rot 17.4025 0.0001 0>}]
 dyn_hair_back_l_01_MSTR [{shapeType=Capsule, length=0.505, radius=0.505, shapeName=dyn_hair_back_l_01_MSTR_capsule, rotation=<rot 0 -90 0>}]
 dyn_hair_back_l_02 [{
   shapeType=Capsule
   length=1
   radius=1
   shapeName=dyn_hair_cowlick_r_03_capsule
   center=<1.7641 -0.6028 0.0874>
   rotation=<rot 12.5593 -2.0115 72.2192>
  }]
 dyn_hair_back_l_02_MSTR [{shapeType=Capsule, length=0.505, radius=0.505, shapeName=dyn_hair_back_l_02_MSTR_capsule, rotation=<rot 0 -90 0>}]
 dyn_hair_back_r_00 [{
   shapeType=Capsule
   length=1
   radius=1
   shapeName=dyn_hair_cowlick_r_03_capsule
   center=<1.9287 0 0.6045>
   rotation=<rot 17.4024 0.0001 24.4751>
  }]
 dyn_hair_back_r_00_MSTR [{shapeType=Capsule, length=0.505, radius=0.505, shapeName=dyn_hair_back_r_00_MSTR_capsule, rotation=<rot 0 -90 0>}]
 dyn_hair_back_r_01 [{shapeType=Capsule, length=1, radius=1, shapeName=dyn_hair_cowlick_r_03_capsule, center=<1.9119 0 0.5992>, rotation=<rot 17.4025 0.0001 0>}]
 dyn_hair_back_r_01_MSTR [{shapeType=Capsule, length=0.505, radius=0.505, shapeName=dyn_hair_back_r_01_MSTR_capsule, rotation=<rot 0 -90 0>}]
 dyn_hair_cowlick_l_00 [{
   shapeType=Capsule
   length=1
   radius=1
   shapeName=dyn_hair_cowlick_r_03_capsule
   center=<1.8685 0.3023 0.4675>
   rotation=<rot 16.9232 13.8875 82.5222>
  }]
 dyn_hair_cowlick_l_00_MSTR [{shapeType=Capsule, length=0.505, radius=0.505, shapeName=dyn_hair_cowlick_l_00_MSTR_capsule, rotation=<rot 0 -90 0>}]
 dyn_hair_cowlick_l_01 [{shapeType=Capsule, length=1, radius=1, shapeName=dyn_hair_cowlick_r_00_capsule, center=<1.7491 -0.3488 1.1909>, rotation=<rot 0 0 159.5447>}]
 dyn_hair_cowlick_l_01_MSTR [{shapeType=Capsule, length=0.505, radius=0.505, shapeName=dyn_hair_cowlick_l_01_MSTR_capsule, rotation=<rot 0 -90 0>}]
 dyn_hair_cowlick_l_02 [{
   shapeType=Capsule
   length=1
   radius=1
   shapeName=dyn_hair_cowlick_r_03_capsule
   center=<2.4519 0 0.4241>
   rotation=<rot 17.4025 0.0001 66.5895>
  }]
 dyn_hair_cowlick_l_02_MSTR [{shapeType=Capsule, length=0.505, radius=0.505, shapeName=dyn_hair_cowlick_l_02_MSTR_capsule, rotation=<rot 0 -90 0>}]
 dyn_hair_cowlick_l_03 [{
   shapeType=Capsule
   length=1
   radius=1
   shapeName=dyn_hair_cowlick_r_03_capsule
   center=<2.3989 0 0.126>
   rotation=<rot 5.7149 0.0002 -0.0001>
  }]
 dyn_hair_cowlick_l_03_MSTR [{shapeType=Capsule, length=0.505, radius=0.505, shapeName=dyn_hair_cowlick_l_03_MSTR_capsule, rotation=<rot 0 -90 0>}]
 dyn_hair_cowlick_r_00_MSTR_Rasp [{shapeType=Capsule, length=0.505, radius=0.505, shapeName=dyn_hair_cowlick_r_00_MSTR_Rasp_capsule, rotation=<rot 0 -90 0>}]
 dyn_hair_cowlick_r_00_Rasp [{
   shapeType=Capsule
   length=1.5
   radius=1
   shapeName=dyn_hair_cowlick_r_00_Rasp_capsule
   center=<2.3186 -0.0544 -0.1785>
   rotation=<rot 3.3648 -176.992 89.8969>
  }]
 dyn_hair_cowlick_r_01 [{
   shapeType=Capsule
   length=1
   radius=1
   shapeName=dyn_hair_cowlick_r_03_capsule
   center=<3.4558 0 0.7387>
   rotation=<rot 17.4024 0.0001 -108.4757>
  }]
 dyn_hair_cowlick_r_01_MSTR [{shapeType=Capsule, length=0.505, radius=0.505, shapeName=dyn_hair_cowlick_r_01_MSTR_capsule, rotation=<rot 0 -90 0>}]
 dyn_hair_cowlick_r_02 [{
   shapeType=Capsule
   length=1
   radius=1
   shapeName=dyn_hair_cowlick_r_03_capsule
   center=<2.5614 0.1585 0.0748>
   rotation=<rot 17.4025 0.0001 66.5895>
  }]
 dyn_hair_cowlick_r_02_MSTR [{shapeType=Capsule, length=0.505, radius=0.505, shapeName=dyn_hair_cowlick_r_02_MSTR_capsule, rotation=<rot 0 -90 0>}]
 dyn_hair_cowlick_r_03 [{shapeType=Capsule, length=1, radius=1, shapeName=dyn_hair_cowlick_r_03_capsule, center=<1.347 0 0.4222>, rotation=<rot 17.4025 0.0001 0>}]
 dyn_hair_cowlick_r_03_MSTR [{shapeType=Capsule, length=0.505, radius=0.505, shapeName=dyn_hair_cowlick_r_03_MSTR_capsule, rotation=<rot 0 -90 0>}]
 dyn_hair_cowlick_r_04 [{shapeType=Capsule, length=1, radius=1, shapeName=dyn_hair_cowlick_r_03_capsule, center=<2.5992 0 0.2507>, rotation=<rot 17.4024 0 -77.5247>}]
 dyn_hair_cowlick_r_04_MSTR [{shapeType=Capsule, length=0.505, radius=0.505, shapeName=dyn_hair_cowlick_r_04_MSTR_capsule, rotation=<rot 0 -90 0>}]
 dyn_hair_cowlick_r_05 [{
   shapeType=Capsule
   length=1
   radius=1
   shapeName=dyn_hair_cowlick_r_03_capsule
   center=<2.4519 0 0.4241>
   rotation=<rot 17.4025 0.0001 66.5895>
  }]
 dyn_hair_cowlick_r_05_MSTR [{shapeType=Capsule, length=0.505, radius=0.505, shapeName=dyn_hair_cowlick_r_05_MSTR_capsule, rotation=<rot 0 -90 0>}]
 dyn_hair_front_00_MSTR_Rasp [{shapeType=Capsule, length=0.505, radius=0.505, shapeName=dyn_hair_front_00_MSTR_Rasp_capsule, rotation=<rot 0 -90 0>}]
 dyn_hair_front_00_Rasp [{shapeType=Capsule, length=1.5, radius=0.75, shapeName=dyn_hair_front_00_Rasp_capsule, center=<0.7183 0 0>}]
 dyn_hair_front_01_Rasp [{shapeType=Capsule, length=1.5, radius=0.75, shapeName=dyn_hair_front_01_Rasp_capsule, center=<1.5153 0 0>}]
 dyn_hair_front_02_MSTR_Rasp [{shapeType=Capsule, length=0.505, radius=0.505, shapeName=dyn_hair_front_02_MSTR_Rasp_capsule, rotation=<rot 0 -90 0>}]
 dyn_hair_front_02_Rasp [{
   shapeType=Capsule
   length=1.25
   radius=1
   shapeName=dyn_hair_front_01_Rasp_capsule
   center=<-0.0342 -1.7614 -0.055>
   rotation=<rot -0.4225 2.6227 -167.725>
  }]
 dyn_hair_side_l_00 [{shapeType=Capsule, length=1, radius=1, shapeName=dyn_hair_side_l_00_capsule, center=<1.4784 0 0>, rotation=<rot 0 0 80.9286>}]
 dyn_hair_side_l_00_MSTR [{shapeType=Capsule, length=0.505, radius=0.505, shapeName=dyn_hair_side_l_00_MSTR_capsule, rotation=<rot 0 -90 0>}]
 dyn_hair_side_l_01 [{shapeType=Capsule, length=1, radius=1, shapeName=dyn_hair_side_l_01_capsule, center=<1.4581 0 0>, rotation=<rot 0.0001 0.0002 -70.7462>}]
 dyn_hair_side_l_01_MSTR [{shapeType=Capsule, length=0.505, radius=0.505, shapeName=dyn_hair_side_l_01_MSTR_capsule, rotation=<rot 0 -90 0>}]
 dyn_hair_side_l_02 [{shapeType=Capsule, length=1, radius=1, shapeName=dyn_hair_side_l_02_capsule, center=<2.7246 0 0>, rotation=<rot 0.0002 0.0003 -66.0644>}]
 dyn_hair_side_l_02_MSTR [{shapeType=Capsule, length=0.505, radius=0.505, shapeName=dyn_hair_side_l_02_MSTR_capsule, rotation=<rot 0 -90 0>}]
 dyn_hair_side_l_03_MSTR_Rasp [{shapeType=Capsule, length=0.505, radius=0.505, shapeName=dyn_hair_side_l_03_MSTR_Rasp_capsule, rotation=<rot 0 -90 0>}]
 dyn_hair_side_l_03_Rasp [{
   shapeType=Capsule
   length=1
   radius=1
   shapeName=dyn_hair_cowlick_r_03_capsule
   center=<2.8671 -0.0343 -0.2387>
   rotation=<rot -9.3851 -0.829 91.7496>
  }]
 dyn_hair_side_l_04 [{
   shapeType=Capsule
   length=1
   radius=1
   shapeName=dyn_hair_cowlick_r_03_capsule
   center=<1.7143 0.8373 0.5373>
   rotation=<rot 16.9232 13.8875 82.5222>
  }]
 dyn_hair_side_l_04_MSTR [{shapeType=Capsule, length=0.505, radius=0.505, shapeName=dyn_hair_side_l_04_MSTR_capsule, rotation=<rot 0 -90 0>}]
 dyn_hair_side_l_05_MSTR_Rasp [{shapeType=Capsule, length=0.505, radius=0.505, shapeName=dyn_hair_side_l_05_MSTR_Rasp_capsule, rotation=<rot 0 -90 0>}]
 dyn_hair_side_l_05_Rasp [{shapeType=Capsule, length=1, radius=0.75, shapeName=dyn_hair_side_l_05_Rasp_capsule, center=<1.5 0 0>}]
 dyn_hair_side_r_00 [{
   shapeType=Capsule
   length=1
   radius=1
   shapeName=dyn_hair_cowlick_r_03_capsule
   center=<2.1227 0.2275 0.1585>
   rotation=<rot 5.6878 -2.5834 -78.0446>
  }]
 dyn_hair_side_r_00_MSTR [{shapeType=Capsule, length=0.505, radius=0.505, shapeName=dyn_hair_side_r_00_MSTR_capsule, rotation=<rot 0 -90 0>}]
 dyn_hair_side_r_01 [{shapeType=Capsule, length=1, radius=1, shapeName=dyn_hair_cowlick_l_00_capsule, center=<2.243 0 0>, rotation=<rot -0.0547 17.17 -89.8228>}]
 dyn_hair_side_r_01_MSTR [{shapeType=Capsule, length=0.505, radius=0.505, shapeName=dyn_hair_side_r_01_MSTR_capsule, rotation=<rot 0 -90 0>}]
 dyn_hair_side_r_02 [{shapeType=Capsule, length=1, radius=1, shapeName=dyn_hair_side_l_00_capsule, center=<1.4784 0 0>, rotation=<rot 0 0 80.9286>}]
 dyn_hair_side_r_02_MSTR [{shapeType=Capsule, length=0.505, radius=0.505, shapeName=dyn_hair_side_r_02_MSTR_capsule, rotation=<rot 0 -90 0>}]
 dyn_hair_side_r_03 [{
   shapeType=Capsule
   length=1
   radius=1
   shapeName=dyn_hair_cowlick_r_03_capsule
   center=<1.9119 0 0.5992>
   rotation=<rot 12.929 -16.5117 75.1575>
  }]
 dyn_hair_side_r_03_MSTR [{shapeType=Capsule, length=0.505, radius=0.505, shapeName=dyn_hair_side_r_03_MSTR_capsule, rotation=<rot 0 -90 0>}]
 dyn_hair_side_r_04 [{
   shapeType=Capsule
   length=1
   radius=1
   shapeName=dyn_hair_cowlick_r_03_capsule
   center=<1.9001 -0.0805 0.0974>
   rotation=<rot -0.8502 -3.0355 99.1651>
  }]
 dyn_hair_side_r_04_MSTR [{shapeType=Capsule, length=0.505, radius=0.505, shapeName=dyn_hair_side_r_04_MSTR_capsule, rotation=<rot 0 -90 0>}]
 dyn_hair_side_r_05 [{
   shapeType=Capsule
   length=1
   radius=1
   shapeName=dyn_hair_cowlick_r_03_capsule
   center=<2.8671 -0.0343 -0.2387>
   rotation=<rot -9.3851 -0.829 91.7496>
  }]
 dyn_hair_side_r_05_MSTR [{shapeType=Capsule, length=0.505, radius=0.505, shapeName=dyn_hair_side_r_05_MSTR_capsule, rotation=<rot 0 -90 0>}]
 dyn_hair_side_r_06_MSTR_Rasp [{shapeType=Capsule, length=0.505, radius=0.505, shapeName=dyn_hair_side_r_06_MSTR_Rasp_capsule, rotation=<rot 0 -90 0>}]
 dyn_hair_side_r_06_Rasp [{
   shapeType=Capsule
   length=1
   radius=1
   shapeName=dyn_hair_cowlick_r_03_capsule
   center=<2.5614 0.1585 0.0748>
   rotation=<rot 17.4025 0.0001 66.5895>
  }]
 dyn_hair_side_r_07 [{shapeType=Capsule, length=1, radius=1, shapeName=dyn_hair_front_00_capsule, center=<3.6319 0 0>}]
 dyn_hair_side_r_07_MSTR [{shapeType=Capsule, length=0.505, radius=0.505, shapeName=dyn_hair_side_r_07_MSTR_capsule, rotation=<rot 0 -90 0>}]
 dyn_hair_top_cowlick [{shapeType=Capsule, length=1, radius=1, shapeName=dyn_hair_top_cowlick_capsule, center=<0 0 1.8247>, rotation=<rot 0.0098 -44.7577 90.0001>}]
 dyn_hair_top_cowlick_MSTR [{shapeType=Capsule, length=0.505, radius=0.505, shapeName=dyn_hair_top_cowlick_MSTR_capsule, rotation=<rot 0 -90 0>}]
 head [
  {shapeType=Capsule, length=7.105, radius=5.0542, center=<5.1382 1.0792 -0.2886>, rotation=<rot 86.7858 -0.0001 -23.3708>}
  {shapeType=Capsule, length=5.75, radius=2.125, shapeName=head_capsule, center=<2.3571 8.1979 0.2857>, rotation=<rot 1.9297 75.1777 -90.5104>}
 ]
}
constraints [
 {bone1Name=dyn_hair_top_cowlick, bone2Name=dyn_hair_top_cowlick_MSTR, swing1LimitDegrees=10, swing2LimitDegrees=10, twistLimitDegrees=5}
 {bone1Name=dyn_hair_side_l_02, bone2Name=dyn_hair_side_l_02_MSTR, swing1LimitDegrees=5, swing2LimitDegrees=5, twistLimitDegrees=2}
 {bone1Name=dyn_hair_side_l_01, bone2Name=dyn_hair_side_l_01_MSTR, swing1LimitDegrees=5, swing2LimitDegrees=5, twistLimitDegrees=2}
 {bone1Name=dyn_hair_side_l_00, bone2Name=dyn_hair_side_l_00_MSTR, swing1LimitDegrees=5, swing2LimitDegrees=5, twistLimitDegrees=2}
 {bone1Name=dyn_hair_cowlick_r_03, bone2Name=dyn_hair_cowlick_r_03_MSTR, swing1LimitDegrees=10, swing2LimitDegrees=10, twistLimitDegrees=5}
 {bone1Name=dyn_hair_back_l_01, bone2Name=dyn_hair_back_l_01_MSTR, swing1LimitDegrees=5, swing2LimitDegrees=5, twistLimitDegrees=5}
 {bone1Name=dyn_hair_back_l_02, bone2Name=dyn_hair_back_l_02_MSTR, swing1LimitDegrees=10, swing2LimitDegrees=10, twistLimitDegrees=5}
 {bone1Name=dyn_hair_back_r_00, bone2Name=dyn_hair_back_r_00_MSTR, swing1LimitDegrees=10, swing2LimitDegrees=10, twistLimitDegrees=5}
 {bone1Name=dyn_hair_back_r_01, bone2Name=dyn_hair_back_r_01_MSTR, swing1LimitDegrees=10, swing2LimitDegrees=10, twistLimitDegrees=2}
 {bone1Name=dyn_hair_cowlick_r_01, bone2Name=dyn_hair_cowlick_r_01_MSTR, swing1LimitDegrees=10, swing2LimitDegrees=10, twistLimitDegrees=5}
 {bone1Name=dyn_hair_cowlick_r_02, bone2Name=dyn_hair_cowlick_r_02_MSTR, swing1LimitDegrees=5, swing2LimitDegrees=5, twistLimitDegrees=2}
 {bone1Name=dyn_hair_backTop_00, bone2Name=dyn_hair_backTop_00_MSTR, swing1LimitDegrees=5, swing2LimitDegrees=5, twistLimitDegrees=2}
 {bone1Name=dyn_hair_side_r_04, bone2Name=dyn_hair_side_r_04_MSTR, swing1LimitDegrees=10, swing2LimitDegrees=10, twistLimitDegrees=5}
 {bone1Name=dyn_hair_cowlick_r_04, bone2Name=dyn_hair_cowlick_r_04_MSTR, swing1LimitDegrees=10, swing2LimitDegrees=10, twistLimitDegrees=5}
 {bone1Name=dyn_hair_side_l_04, bone2Name=dyn_hair_side_l_04_MSTR, swing1LimitDegrees=10, swing2LimitDegrees=10, twistLimitDegrees=5}
 {bone1Name=dyn_hair_cowlick_l_01, bone2Name=dyn_hair_cowlick_l_01_MSTR, swing1LimitDegrees=15, swing2LimitDegrees=15, twistLimitDegrees=5}
 {bone1Name=dyn_hair_cowlick_l_03, bone2Name=dyn_hair_cowlick_l_03_MSTR, swing1LimitDegrees=20, swing2LimitDegrees=20, twistLimitDegrees=5}
 {bone1Name=dyn_hair_side_r_07, bone2Name=dyn_hair_side_r_07_MSTR, swing1LimitDegrees=10, swing2LimitDegrees=10, twistLimitDegrees=5}
 {bone1Name=dyn_hair_side_r_05, bone2Name=dyn_hair_side_r_05_MSTR, swing1LimitDegrees=10, swing2LimitDegrees=10, twistLimitDegrees=5}
 {bone1Name=dyn_hair_side_r_02, bone2Name=dyn_hair_side_r_02_MSTR, swing1LimitDegrees=5, swing2LimitDegrees=5, twistLimitDegrees=2}
 {bone1Name=dyn_hair_side_r_00, bone2Name=dyn_hair_side_r_00_MSTR, swing1LimitDegrees=5, swing2LimitDegrees=5, twistLimitDegrees=2}
 {bone1Name=dyn_hair_side_r_03, bone2Name=dyn_hair_side_r_03_MSTR, swing1LimitDegrees=5, swing2LimitDegrees=5, twistLimitDegrees=2}
 {bone1Name=dyn_hair_side_r_01, bone2Name=dyn_hair_side_r_01_MSTR, swing1LimitDegrees=10, swing2LimitDegrees=10, twistLimitDegrees=5}
 {bone1Name=dyn_hair_cowlick_l_02, bone2Name=dyn_hair_cowlick_l_02_MSTR, swing1LimitDegrees=10, swing2LimitDegrees=10, twistLimitDegrees=5}
 {bone1Name=dyn_hair_cowlick_l_00, bone2Name=dyn_hair_cowlick_l_00_MSTR, swing1LimitDegrees=10, swing2LimitDegrees=10, twistLimitDegrees=5}
 {bone1Name=dyn_hair_backTop_01, bone2Name=dyn_hair_backTop_01_MSTR, swing1LimitDegrees=5, swing2LimitDegrees=5, twistLimitDegrees=2}
 {bone1Name=dyn_hair_cowlick_r_05, bone2Name=dyn_hair_cowlick_r_05_MSTR, swing1LimitDegrees=5, swing2LimitDegrees=5, twistLimitDegrees=2}
 {bone1Name=dyn_hair_front_00_Rasp, bone2Name=dyn_hair_front_00_MSTR_Rasp, swing1LimitDegrees=5, swing2LimitDegrees=5, twistLimitDegrees=2}
 {bone1Name=dyn_hair_front_01_Rasp, bone2Name=dyn_hair_front_00_Rasp, swing1LimitDegrees=5, swing2LimitDegrees=5, twistLimitDegrees=2}
 {
  bone1Name=dyn_hair_backTop_00_MSTR
  bone2Name=head
  swing1LimitDegrees=45
  swing1Motion=Free
  swing2LimitDegrees=45
  swing2Motion=Free
  twistLimitDegrees=45
  twistMotion=Free
 }
 {
  bone1Name=dyn_hair_backTop_01_MSTR
  bone2Name=head
  swing1LimitDegrees=45
  swing1Motion=Free
  swing2LimitDegrees=45
  swing2Motion=Free
  twistLimitDegrees=45
  twistMotion=Free
 }
 {
  bone1Name=dyn_hair_back_l_01_MSTR
  bone2Name=head
  swing1LimitDegrees=45
  swing1Motion=Free
  swing2LimitDegrees=45
  swing2Motion=Free
  twistLimitDegrees=45
  twistMotion=Free
 }
 {
  bone1Name=dyn_hair_back_l_02_MSTR
  bone2Name=head
  swing1LimitDegrees=45
  swing1Motion=Free
  swing2LimitDegrees=45
  swing2Motion=Free
  twistLimitDegrees=45
  twistMotion=Free
 }
 {
  bone1Name=dyn_hair_back_r_00_MSTR
  bone2Name=head
  swing1LimitDegrees=45
  swing1Motion=Free
  swing2LimitDegrees=45
  swing2Motion=Free
  twistLimitDegrees=45
  twistMotion=Free
 }
 {
  bone1Name=dyn_hair_back_r_01_MSTR
  bone2Name=head
  swing1LimitDegrees=45
  swing1Motion=Free
  swing2LimitDegrees=45
  swing2Motion=Free
  twistLimitDegrees=45
  twistMotion=Free
 }
 {
  bone1Name=dyn_hair_cowlick_l_00_MSTR
  bone2Name=head
  swing1LimitDegrees=45
  swing1Motion=Free
  swing2LimitDegrees=45
  swing2Motion=Free
  twistLimitDegrees=45
  twistMotion=Free
 }
 {
  bone1Name=dyn_hair_cowlick_l_01_MSTR
  bone2Name=head
  swing1LimitDegrees=45
  swing1Motion=Free
  swing2LimitDegrees=45
  swing2Motion=Free
  twistLimitDegrees=45
  twistMotion=Free
 }
 {
  bone1Name=dyn_hair_cowlick_l_02_MSTR
  bone2Name=head
  swing1LimitDegrees=45
  swing1Motion=Free
  swing2LimitDegrees=45
  swing2Motion=Free
  twistLimitDegrees=45
  twistMotion=Free
 }
 {
  bone1Name=dyn_hair_cowlick_l_03_MSTR
  bone2Name=head
  swing1LimitDegrees=45
  swing1Motion=Free
  swing2LimitDegrees=45
  swing2Motion=Free
  twistLimitDegrees=45
  twistMotion=Free
 }
 {
  bone1Name=dyn_hair_cowlick_r_01_MSTR
  bone2Name=head
  swing1LimitDegrees=45
  swing1Motion=Free
  swing2LimitDegrees=45
  swing2Motion=Free
  twistLimitDegrees=45
  twistMotion=Free
 }
 {
  bone1Name=dyn_hair_cowlick_r_02_MSTR
  bone2Name=head
  swing1LimitDegrees=45
  swing1Motion=Free
  swing2LimitDegrees=45
  swing2Motion=Free
  twistLimitDegrees=45
  twistMotion=Free
 }
 {
  bone1Name=dyn_hair_cowlick_r_03_MSTR
  bone2Name=head
  swing1LimitDegrees=45
  swing1Motion=Free
  swing2LimitDegrees=45
  swing2Motion=Free
  twistLimitDegrees=45
  twistMotion=Free
 }
 {
  bone1Name=dyn_hair_cowlick_r_04_MSTR
  bone2Name=head
  swing1LimitDegrees=45
  swing1Motion=Free
  swing2LimitDegrees=45
  swing2Motion=Free
  twistLimitDegrees=45
  twistMotion=Free
 }
 {
  bone1Name=dyn_hair_cowlick_r_05_MSTR
  bone2Name=head
  swing1LimitDegrees=45
  swing1Motion=Free
  swing2LimitDegrees=45
  swing2Motion=Free
  twistLimitDegrees=45
  twistMotion=Free
 }
 {
  bone1Name=dyn_hair_front_00_MSTR_Rasp
  bone2Name=head
  swing1LimitDegrees=45
  swing1Motion=Free
  swing2LimitDegrees=45
  swing2Motion=Free
  twistLimitDegrees=45
  twistMotion=Free
 }
 {
  bone1Name=dyn_hair_side_l_00_MSTR
  bone2Name=head
  swing1LimitDegrees=45
  swing1Motion=Free
  swing2LimitDegrees=45
  swing2Motion=Free
  twistLimitDegrees=45
  twistMotion=Free
 }
 {
  bone1Name=dyn_hair_side_l_01_MSTR
  bone2Name=head
  swing1LimitDegrees=45
  swing1Motion=Free
  swing2LimitDegrees=45
  swing2Motion=Free
  twistLimitDegrees=45
  twistMotion=Free
 }
 {
  bone1Name=dyn_hair_side_l_02_MSTR
  bone2Name=head
  swing1LimitDegrees=45
  swing1Motion=Free
  swing2LimitDegrees=45
  swing2Motion=Free
  twistLimitDegrees=45
  twistMotion=Free
 }
 {
  bone1Name=dyn_hair_side_l_04_MSTR
  bone2Name=head
  swing1LimitDegrees=45
  swing1Motion=Free
  swing2LimitDegrees=45
  swing2Motion=Free
  twistLimitDegrees=45
  twistMotion=Free
 }
 {
  bone1Name=dyn_hair_side_r_00_MSTR
  bone2Name=head
  swing1LimitDegrees=45
  swing1Motion=Free
  swing2LimitDegrees=45
  swing2Motion=Free
  twistLimitDegrees=45
  twistMotion=Free
 }
 {
  bone1Name=dyn_hair_side_r_01_MSTR
  bone2Name=head
  swing1LimitDegrees=45
  swing1Motion=Free
  swing2LimitDegrees=45
  swing2Motion=Free
  twistLimitDegrees=45
  twistMotion=Free
 }
 {
  bone1Name=dyn_hair_side_r_02_MSTR
  bone2Name=head
  swing1LimitDegrees=45
  swing1Motion=Free
  swing2LimitDegrees=45
  swing2Motion=Free
  twistLimitDegrees=45
  twistMotion=Free
 }
 {
  bone1Name=dyn_hair_side_r_03_MSTR
  bone2Name=head
  swing1LimitDegrees=45
  swing1Motion=Free
  swing2LimitDegrees=45
  swing2Motion=Free
  twistLimitDegrees=45
  twistMotion=Free
 }
 {
  bone1Name=dyn_hair_side_r_04_MSTR
  bone2Name=head
  swing1LimitDegrees=45
  swing1Motion=Free
  swing2LimitDegrees=45
  swing2Motion=Free
  twistLimitDegrees=45
  twistMotion=Free
 }
 {
  bone1Name=dyn_hair_side_r_05_MSTR
  bone2Name=head
  swing1LimitDegrees=45
  swing1Motion=Free
  swing2LimitDegrees=45
  swing2Motion=Free
  twistLimitDegrees=45
  twistMotion=Free
 }
 {
  bone1Name=dyn_hair_side_r_07_MSTR
  bone2Name=head
  swing1LimitDegrees=45
  swing1Motion=Free
  swing2LimitDegrees=45
  swing2Motion=Free
  twistLimitDegrees=45
  twistMotion=Free
 }
 {
  bone1Name=dyn_hair_top_cowlick_MSTR
  bone2Name=head
  swing1LimitDegrees=45
  swing1Motion=Free
  swing2LimitDegrees=45
  swing2Motion=Free
  twistLimitDegrees=45
  twistMotion=Free
 }
 {
  bone1Name=dyn_hair_front_02_MSTR_Rasp
  bone2Name=head
  swing1LimitDegrees=45
  swing1Motion=Free
  swing2LimitDegrees=45
  swing2Motion=Free
  twistLimitDegrees=45
  twistMotion=Free
 }
 {bone1Name=dyn_hair_front_02_Rasp, bone2Name=dyn_hair_front_02_MSTR_Rasp, swing1LimitDegrees=5, swing2LimitDegrees=5, twistLimitDegrees=2}
 {
  bone1Name=dyn_hair_side_l_03_MSTR_Rasp
  bone2Name=head
  swing1LimitDegrees=45
  swing1Motion=Free
  swing2LimitDegrees=45
  swing2Motion=Free
  twistLimitDegrees=45
  twistMotion=Free
 }
 {bone1Name=dyn_hair_side_l_03_Rasp, bone2Name=dyn_hair_side_l_03_MSTR_Rasp, swing1LimitDegrees=10, swing2LimitDegrees=10, twistLimitDegrees=5}
 {
  bone1Name=dyn_hair_cowlick_r_00_MSTR_Rasp
  bone2Name=head
  swing1LimitDegrees=45
  swing1Motion=Free
  swing2LimitDegrees=45
  swing2Motion=Free
  twistLimitDegrees=45
  twistMotion=Free
 }
 {
  bone1Name=dyn_hair_cowlick_r_00_Rasp
  bone2Name=dyn_hair_cowlick_r_00_MSTR_Rasp
  swing1LimitDegrees=5
  swing2LimitDegrees=5
  twistLimitDegrees=3
 }
 {
  bone1Name=dyn_hair_back_l_00_MSTR_Rasp
  bone2Name=head
  swing1LimitDegrees=45
  swing1Motion=Free
  swing2LimitDegrees=45
  swing2Motion=Free
  twistLimitDegrees=45
  twistMotion=Free
 }
 {
  bone1Name=dyn_hair_side_l_05_MSTR_Rasp
  bone2Name=head
  swing1LimitDegrees=45
  swing1Motion=Free
  swing2LimitDegrees=45
  swing2Motion=Free
  twistLimitDegrees=45
  twistMotion=Free
 }
 {bone1Name=dyn_hair_side_l_05_Rasp, bone2Name=dyn_hair_side_l_05_MSTR_Rasp, swing1LimitDegrees=10, swing2LimitDegrees=10, twistLimitDegrees=2}
 {bone1Name=dyn_hair_back_l_00_Rasp, bone2Name=dyn_hair_back_l_00_MSTR_Rasp, swing1LimitDegrees=5, swing2LimitDegrees=5, twistLimitDegrees=5}
 {
  bone1Name=dyn_hair_side_r_06_MSTR_Rasp
  bone2Name=head
  swing1LimitDegrees=45
  swing1Motion=Free
  swing2LimitDegrees=45
  swing2Motion=Free
  twistLimitDegrees=45
  twistMotion=Free
 }
 {bone1Name=dyn_hair_side_r_06_Rasp, bone2Name=dyn_hair_side_r_06_MSTR_Rasp, swing1LimitDegrees=10, swing2LimitDegrees=10, twistLimitDegrees=2}
]
